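# Physics asset for the Future Samurai Cyborg male medium body's dynamic cape, back jacket, chest pad belt and bag belts.
# Capsules on pelvis, spine and thighs form body volumes, while the dangling pieces use mostly box chains.
# MSTR root bodies attach to spine or pelvis with free angular motion, and most chain links have swing and twist limits.
# Cross constraints with free rotation tie left and right cape chains together, front and back.
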PhysicsAsset /BRCosmetics/Characters/Player/Male/Medium/Bodies/M_MED_Future_Samurai_Cyborg/Meshes/M_MED_Future_Samurai_Cyborg_dyn_cape
bodies {
 dyn_backJacket_l [{shapeType=Box, extentX=8, extentY=15, extentZ=1.5, center=<6 0 0>}]
 dyn_backJacket_l_MSTR [{shapeType=Capsule, length=0.505, radius=0.505, rotation=<rot 0 -90 0>}]
 dyn_backJacket_r [{shapeType=Box, extentX=8, extentY=15, extentZ=1.5, center=<6 0 0>}]
 dyn_backJacket_r_MSTR [{shapeType=Capsule, length=0.505, radius=0.505, rotation=<rot 0 -90 0>}]
 dyn_bagBelt_01_l [{shapeType=Box, extentX=5, extentY=4.5, extentZ=1, center=<2 0 0>}]
 dyn_bagBelt_01_r [{shapeType=Box, extentX=5, extentY=2.5, extentZ=1, center=<2 0 0>}]
 dyn_bagBelt_02_l [{shapeType=Box, extentX=6, extentY=4.5, extentZ=1, center=<5 0 -0.5>, rotation=<rot -10 0 0>}]
 dyn_bagBelt_02_r [{shapeType=Box, extentX=6, extentY=2.5, extentZ=1, center=<5 0 -0.5>, rotation=<rot -10 0 5.0001>}]
 dyn_bag_l_MSTR [{shapeType=Box, extentX=5.5, extentY=6, extentZ=10.5, center=<0.433 0.0088 -2.2297>, rotation=<rot 14.9614 -0.0777 -0.8874>}]
 dyn_bag_r_MSTR [{shapeType=Box, extentX=5.5, extentY=6, extentZ=10.5, center=<0.433 0.0088 -2.2297>, rotation=<rot 14.9614 -0.0777 -0.8874>}]
 dyn_chestPad_r [{shapeType=Capsule, length=0.5, radius=0.5}]
 dyn_chestPad_r_MSTR [{shapeType=Capsule, length=0.5, radius=0.5}]
 dyn_leftCape_01_ft [
  {shapeType=Box, extentX=12, extentY=6, extentZ=2, center=<8 -2 2>, rotation=<rot 0 0 60>}
  {shapeType=Box, extentX=12, extentY=10, extentZ=2, center=<8 2 0>}
 ]
 dyn_leftCape_02_ft [
  {shapeType=Box, extentX=12, extentY=6, extentZ=2, center=<8 -2 2>, rotation=<rot 0 0 60>}
  {shapeType=Box, extentX=12, extentY=10, extentZ=2, center=<8 2 0>}
 ]
 dyn_leftCape_03_ft [
  {shapeType=Box, extentX=12, extentY=6, extentZ=2, center=<8 -2 2>, rotation=<rot 0 0 60>}
  {shapeType=Box, extentX=12, extentY=10, extentZ=2, center=<8 2 0>}
 ]
 dyn_leftCape_04_ft [
  {shapeType=Box, extentX=12, extentY=6, extentZ=2, center=<8 -2 2>, rotation=<rot 0 0 60>}
  {shapeType=Box, extentX=12, extentY=10, extentZ=2, center=<8 2 0>}
 ]
 dyn_leftCape_ft_MSTR [{shapeType=Capsule, length=0.505, radius=0.505, rotation=<rot 0 -90 0>}]
 dyn_lefttCape_01_bk [{shapeType=Box, extentX=14, extentY=20, extentZ=2, center=<7 0 0>, rotation=<rot 0 0 15.001>}]
 dyn_lefttCape_02_bk [{shapeType=Box, extentX=14, extentY=20, extentZ=2, center=<7 0 0>, rotation=<rot 0 0 15.001>}]
 dyn_lefttCape_03_bk [{shapeType=Box, extentX=14, extentY=20, extentZ=2, center=<7 0 0>, rotation=<rot 0 0 15.001>}]
 dyn_lefttCape_bk_MSTR [{shapeType=Capsule, length=0.505, radius=0.505, rotation=<rot 0 -90 0>}]
 dyn_padBelt_01 [{shapeType=Box, extentX=6.5, extentY=4.5, extentZ=1.5, center=<4 0 0>}]
 dyn_padBelt_02 [{shapeType=Box, extentX=6.5, extentY=4.5, extentZ=1.5, center=<4 0 0>}]
 dyn_padBelt_03 [{shapeType=Box, extentX=3.5, extentY=4.5, extentZ=1.5, center=<2 0 0>}]
 dyn_rightCape_00_bk [{shapeType=Box, extentX=14, extentY=20, extentZ=2, center=<7 0 0>}]
 dyn_rightCape_01_bk [{shapeType=Box, extentX=14, extentY=20, extentZ=2, center=<7 0 0>}]
 dyn_rightCape_01_ft [{shapeType=Box, extentX=12, extentY=15, extentZ=2, center=<8 0 0>}]
 dyn_rightCape_02_bk [{shapeType=Box, extentX=14, extentY=20, extentZ=2, center=<7 0 0>}]
 dyn_rightCape_02_ft [{shapeType=Box, extentX=12, extentY=15, extentZ=2, center=<8 0 0>}]
 dyn_rightCape_03_bk [{shapeType=Box, extentX=14, extentY=20, extentZ=2, center=<7 0 0>}]
 dyn_rightCape_03_ft [{shapeType=Box, extentX=12, extentY=15, extentZ=2, center=<8 0 0>}]
 dyn_rightCape_04_bk [{shapeType=Box, extentX=14, extentY=20, extentZ=2, center=<7 0 0>}]
 dyn_rightCape_04_ft [{shapeType=Box, extentX=12, extentY=15, extentZ=2, center=<8 0 0>}]
 dyn_rightCape_05_bk [{shapeType=Box, extentX=14, extentY=20, extentZ=2, center=<7 0 0>}]
 dyn_rightCape_05_ft [{shapeType=Box, extentX=12, extentY=15, extentZ=2, center=<8 0 0>}]
 dyn_rightCape_bk_MSTR [{shapeType=Capsule, length=0.505, radius=0.505, rotation=<rot 0 -90 0>}]
 dyn_rightCape_ft_MSTR [{shapeType=Capsule, length=0.505, radius=0.505, rotation=<rot 0 -90 0>}]
 pelvis [
  {shapeType=Capsule, length=12.5, radius=12.25, center=<-8.6529 0.6959 0.0001>, rotation=<rot 0 -0.0003 0>}
  {shapeType=Capsule, length=3.5, radius=12, center=<-16.8187 0.7198 0>, rotation=<rot 90 9.7524 14.9463>}
 ]
 spine_01 [{shapeType=Capsule, length=11, radius=13.25, center=<-6.146 -0.3268 0.0003>}]
 spine_02 [{shapeType=Capsule, length=7, radius=13, center=<-2.5898 1.0699 0>}]
 spine_03 [{shapeType=Capsule, length=7, radius=13.5, center=<0 1.3748 0>}]
 spine_04 [{shapeType=Capsule, length=9.5, radius=15.5, center=<0 0.6272 0.0001>}]
 spine_05 [{shapeType=Capsule, length=13.5, radius=14.5, center=<1.724 0.1109 -0.0289>, rotation=<rot 0 0 -2.8706>}]
 thigh_twist_01_l [{shapeType=Capsule, length=30.25, radius=10, center=<2.0326 1.6413 -0.0001>, rotation=<rot 90 0 0>}]
 thigh_twist_01_r [{shapeType=Capsule, length=30.25, radius=10, center=<-2.0326 -1.6413 0.0001>, rotation=<rot 90 0 180>}]
}
constraints [
 {bone1Name=spine_02, bone2Name=spine_01, swing1LimitDegrees=45, swing2LimitDegrees=45, twistLimitDegrees=45}
 {bone1Name=spine_04, bone2Name=spine_03, swing1LimitDegrees=45, swing2LimitDegrees=45, twistLimitDegrees=45}
 {bone1Name=spine_05, bone2Name=spine_04, swing1LimitDegrees=45, swing2LimitDegrees=45, twistLimitDegrees=45}
 {bone1Name=thigh_twist_01_l, bone2Name=pelvis, swing1LimitDegrees=45, swing2LimitDegrees=45, twistLimitDegrees=45}
 {bone1Name=dyn_rightCape_01_bk, bone2Name=dyn_rightCape_00_bk, swing1LimitDegrees=25, swing2LimitDegrees=30, twistLimitDegrees=2}
 {bone1Name=dyn_rightCape_02_bk, bone2Name=dyn_rightCape_01_bk, swing1LimitDegrees=25, swing2LimitDegrees=30, twistLimitDegrees=2}
 {bone1Name=dyn_rightCape_03_bk, bone2Name=dyn_rightCape_02_bk, swing1LimitDegrees=25, swing2LimitDegrees=30, twistLimitDegrees=2}
 {bone1Name=dyn_rightCape_04_bk, bone2Name=dyn_rightCape_03_bk, swing1LimitDegrees=25, swing2LimitDegrees=30, twistLimitDegrees=2}
 {bone1Name=dyn_rightCape_05_bk, bone2Name=dyn_rightCape_04_bk, swing1LimitDegrees=25, swing2LimitDegrees=30, twistLimitDegrees=2}
 {
  bone1Name=dyn_lefttCape_bk_MSTR
  bone2Name=spine_05
  swing1LimitDegrees=45
  swing1Motion=Free
  swing2LimitDegrees=45
  swing2Motion=Free
  twistLimitDegrees=45
  twistMotion=Free
 }
 {bone1Name=dyn_lefttCape_01_bk, bone2Name=dyn_lefttCape_bk_MSTR, swing1LimitDegrees=25, swing2LimitDegrees=30, twistLimitDegrees=2}
 {bone1Name=dyn_lefttCape_02_bk, bone2Name=dyn_lefttCape_01_bk, swing1LimitDegrees=25, swing2LimitDegrees=30, twistLimitDegrees=2}
 {bone1Name=dyn_lefttCape_03_bk, bone2Name=dyn_lefttCape_02_bk, swing1LimitDegrees=25, swing2LimitDegrees=30, twistLimitDegrees=2}
 {
  bone1Name=dyn_leftCape_ft_MSTR
  bone2Name=spine_05
  swing1LimitDegrees=45
  swing1Motion=Free
  swing2LimitDegrees=45
  swing2Motion=Free
  twistLimitDegrees=45
  twistMotion=Free
 }
 {bone1Name=dyn_leftCape_01_ft, bone2Name=dyn_leftCape_ft_MSTR, swing1LimitDegrees=30, swing2LimitDegrees=45, twistLimitDegrees=30}
 {bone1Name=dyn_leftCape_02_ft, bone2Name=dyn_leftCape_01_ft, swing1LimitDegrees=30, swing2LimitDegrees=45, twistLimitDegrees=30}
 {bone1Name=dyn_leftCape_03_ft, bone2Name=dyn_leftCape_02_ft, swing1LimitDegrees=30, swing2LimitDegrees=45, twistLimitDegrees=30}
 {bone1Name=dyn_leftCape_04_ft, bone2Name=dyn_leftCape_03_ft, swing1LimitDegrees=30, swing2LimitDegrees=45, twistLimitDegrees=30}
 {
  bone1Name=dyn_rightCape_ft_MSTR
  bone2Name=spine_05
  swing1LimitDegrees=45
  swing1Motion=Free
  swing2LimitDegrees=45
  swing2Motion=Free
  twistLimitDegrees=45
  twistMotion=Free
 }
 {bone1Name=dyn_rightCape_01_ft, bone2Name=dyn_rightCape_ft_MSTR, swing1LimitDegrees=30, swing2LimitDegrees=45, twistLimitDegrees=30}
 {bone1Name=dyn_rightCape_02_ft, bone2Name=dyn_rightCape_01_ft, swing1LimitDegrees=30, swing2LimitDegrees=45, twistLimitDegrees=30}
 {bone1Name=dyn_rightCape_03_ft, bone2Name=dyn_rightCape_02_ft, swing1LimitDegrees=30, swing2LimitDegrees=45, twistLimitDegrees=30}
 {bone1Name=dyn_rightCape_04_ft, bone2Name=dyn_rightCape_03_ft, swing1LimitDegrees=30, swing2LimitDegrees=45, twistLimitDegrees=30}
 {bone1Name=dyn_rightCape_05_ft, bone2Name=dyn_rightCape_04_ft, swing1LimitDegrees=30, swing2LimitDegrees=45, twistLimitDegrees=30}
 {
  bone1Name=spine_01
  bone2Name=pelvis
  swing1LimitDegrees=45
  swing1Motion=Free
  swing2LimitDegrees=45
  swing2Motion=Free
  twistLimitDegrees=45
  twistMotion=Free
 }
 {
  bone1Name=spine_03
  bone2Name=spine_02
  swing1LimitDegrees=45
  swing1Motion=Free
  swing2LimitDegrees=45
  swing2Motion=Free
  twistLimitDegrees=45
  twistMotion=Free
 }
 {
  bone1Name=dyn_chestPad_r_MSTR
  bone2Name=spine_05
  swing1LimitDegrees=45
  swing1Motion=Free
  swing2LimitDegrees=45
  swing2Motion=Free
  twistLimitDegrees=45
  twistMotion=Free
 }
 {
  bone1Name=dyn_chestPad_r
  bone2Name=dyn_chestPad_r_MSTR
  swing1LimitDegrees=45
  swing1Motion=Free
  swing2LimitDegrees=45
  swing2Motion=Free
  twistLimitDegrees=45
  twistMotion=Free
 }
 {bone1Name=dyn_padBelt_01, bone2Name=dyn_chestPad_r, swing1LimitDegrees=30, swing2LimitDegrees=30, twistLimitDegrees=10}
 {bone1Name=dyn_padBelt_02, bone2Name=dyn_padBelt_01, swing1LimitDegrees=30, swing2LimitDegrees=30, twistLimitDegrees=10}
 {bone1Name=dyn_padBelt_03, bone2Name=dyn_padBelt_02, swing1LimitDegrees=30, swing2LimitDegrees=30, twistLimitDegrees=10}
 {
  bone1Name=dyn_bag_r_MSTR
  bone2Name=spine_02
  swing1LimitDegrees=45
  swing1Motion=Free
  swing2LimitDegrees=45
  swing2Motion=Free
  twistLimitDegrees=45
  twistMotion=Free
 }
 {bone1Name=dyn_bagBelt_01_r, bone2Name=dyn_bag_r_MSTR, swing1LimitDegrees=2, swing2LimitDegrees=20, twistLimitDegrees=2}
 {bone1Name=dyn_bagBelt_02_r, bone2Name=dyn_bagBelt_01_r, swing1LimitDegrees=30, swing2LimitDegrees=30, twistLimitDegrees=2}
 {
  bone1Name=dyn_bag_l_MSTR
  bone2Name=spine_02
  swing1LimitDegrees=45
  swing1Motion=Free
  swing2LimitDegrees=45
  swing2Motion=Free
  twistLimitDegrees=45
  twistMotion=Free
 }
 {bone1Name=dyn_bagBelt_01_l, bone2Name=dyn_bag_l_MSTR, swing1LimitDegrees=2, swing2LimitDegrees=20, twistLimitDegrees=2}
 {bone1Name=dyn_bagBelt_02_l, bone2Name=dyn_bagBelt_01_l, swing1LimitDegrees=30, swing2LimitDegrees=30, twistLimitDegrees=2}
 {
  bone1Name=dyn_rightCape_bk_MSTR
  bone2Name=spine_05
  swing1LimitDegrees=45
  swing1Motion=Free
  swing2LimitDegrees=45
  swing2Motion=Free
  twistLimitDegrees=45
  twistMotion=Free
 }
 {bone1Name=dyn_rightCape_00_bk, bone2Name=dyn_rightCape_bk_MSTR, swing1LimitDegrees=25, swing2LimitDegrees=30, twistLimitDegrees=2}
 {
  bone1Name=dyn_leftCape_04_ft
  bone2Name=dyn_rightCape_04_ft
  swing1LimitDegrees=15
  swing1Motion=Free
  swing2LimitDegrees=15
  swing2Motion=Free
  twistLimitDegrees=5
  twistMotion=Free
 }
 {
  bone1Name=dyn_lefttCape_03_bk
  bone2Name=dyn_rightCape_03_bk
  swing1LimitDegrees=45
  swing1Motion=Free
  swing2LimitDegrees=45
  swing2Motion=Free
  twistLimitDegrees=45
  twistMotion=Free
 }
 {bone1Name=thigh_twist_01_r, bone2Name=pelvis, swing1LimitDegrees=45, swing2LimitDegrees=45, twistLimitDegrees=45}
 {
  bone1Name=dyn_backJacket_l_MSTR
  bone2Name=pelvis
  swing1LimitDegrees=45
  swing1Motion=Free
  swing2LimitDegrees=45
  swing2Motion=Free
  twistLimitDegrees=45
  twistMotion=Free
 }
 {bone1Name=dyn_backJacket_l, bone2Name=dyn_backJacket_l_MSTR, swing1LimitDegrees=15, swing2LimitDegrees=35, twistLimitDegrees=2}
 {
  bone1Name=dyn_backJacket_r_MSTR
  bone2Name=pelvis
  swing1LimitDegrees=45
  swing1Motion=Free
  swing2LimitDegrees=45
  swing2Motion=Free
  twistLimitDegrees=45
  twistMotion=Free
 }
 {bone1Name=dyn_backJacket_r, bone2Name=dyn_backJacket_r_MSTR, swing1LimitDegrees=15, swing2LimitDegrees=35, twistLimitDegrees=2}
]
