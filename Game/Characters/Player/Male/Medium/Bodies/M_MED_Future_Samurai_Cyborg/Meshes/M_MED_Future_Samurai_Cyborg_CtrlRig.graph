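# Control Rig for the Future Samurai Cyborg body that procedurally drives its shoulder belt, back pad and chest pad bones.
# One execution chain blends TGT and some MSTR helper bones between spine, clavicle and upper-arm twist bones using weighted, offset-maintaining transform constraints.
# AimBone units then point each pad and first shoulder-belt bone at its TGT, after nudging the belt targets along X.
# Second shoulder-belt and pad-belt bones are not driven here.
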ControlRigBlueprint /Game/Characters/Player/Male/Medium/Bodies/M_MED_Future_Samurai_Cyborg/Meshes/M_MED_Future_Samurai_Cyborg_CtrlRig
graphs [
 RigVMFunctionLibrary {links [], nodes []}
 RigVMModel {nodes [
   RigUnit_BeginExecution = &RigUnit_BeginExecution()
   RigUnit_GetBoneTransform = &RigUnit_GetBoneTransform(Bone=spine_05)
   /Script/ControlRig.RigUnit_TransformConstraint(
    BaseTransformSpace=GlobalSpace
    Bone=dyn_shouderBelt_01_r_TGT
    Targets="((Transform=(Rotation=(X=0,Y=0,Z=0,W=1),Translation=(X=0,Y=0,Z=0),Scale3D=(X=1,Y=1,Z=1)),Weight=0.2,bMaintainOffset=True,Filter=(TranslationFilter=(bX=True,bY=True,bZ=True),RotationFilter=(bX=True,bY=True,bZ=True),ScaleFilter=(bX=True,bY=True,bZ=True))),(T ...(truncated)"
    bUseInitialTransforms=True
    ExecutePin=@RigUnit_BeginExecution.ExecutePin
    Targets.0.Transform=@RigUnit_GetBoneTransform.Transform
    Targets.1.Transform=&RigUnit_GetBoneTransform(Bone=clavicle_r).Transform
    BaseTransform=&RigUnit_GetBoneTransform(Bone=dyn_shouderBelt_01_r_TGT).Transform
   ) {is=struct RigVMUnitNode}
   > &RigUnit_TransformConstraint(
    BaseTransformSpace=GlobalSpace
    Bone=dyn_shouderBelt_01_r_MSTR
    Targets="((Transform=(Rotation=(X=0,Y=0,Z=0,W=1),Translation=(X=0,Y=0,Z=0),Scale3D=(X=1,Y=1,Z=1)),Weight=0.8,bMaintainOffset=True,Filter=(TranslationFilter=(bX=True,bY=True,bZ=True),RotationFilter=(bX=True,bY=True,bZ=True),ScaleFilter=(bX=True,bY=True,bZ=True))),(T ...(truncated)"
    bUseInitialTransforms=True
    BaseTransform=&RigUnit_GetBoneTransform(Bone=dyn_shouderBelt_01_r_MSTR).Transform
    Targets.0.Transform=&RigUnit_GetBoneTransform(Bone=spine_05).Transform
    Targets.1.Transform=&RigUnit_GetBoneTransform(Bone=clavicle_r).Transform
   )
   > /Script/ControlRig.RigUnit_AddBoneTransform(
    Bone=dyn_shouderBelt_01_r_TGT
    Transform=(Rotation=(X=0,Y=0,Z=0,W=1),Translation=(X=10,Y=0,Z=0),Scale3D=(X=1,Y=1,Z=1))
    Weight=1
    bPropagateToChildren=True
   ) {is=struct RigVMUnitNode}
   > /Script/ControlRig.RigUnit_AimBone(
    Bone=dyn_shouderBelt_01_r
    DebugSettings=(bEnabled=true,Scale=10,WorldOffset=(Rotation=(X=0,Y=0,Z=0,W=1),Translation=(X=0,Y=0,Z=0),Scale3D=(X=1,Y=1,Z=1)))
    Primary=(Weight=1,Axis=(X=1,Y=0,Z=0),Target=(X=1,Y=0,Z=0),Kind=Location,Space="dyn_shouderBelt_01_r_TGT")
    Secondary=(Weight=1,Axis=(X=0,Y=1,Z=0),Target=(X=0,Y=1,Z=0),Kind=Direction,Space="dyn_shouderBelt_01_r_TGT")
    Weight=1
    bPropagateToChildren=True
   ) {is=struct RigVMUnitNode}
   > &RigUnit_TransformConstraint(
    BaseTransformSpace=GlobalSpace
    Bone=dyn_shouderBelt_01_l_TGT
    Targets="((Transform=(Rotation=(X=0,Y=0,Z=0,W=1),Translation=(X=0,Y=0,Z=0),Scale3D=(X=1,Y=1,Z=1)),Weight=0.1,bMaintainOffset=True,Filter=(TranslationFilter=(bX=True,bY=True,bZ=True),RotationFilter=(bX=True,bY=True,bZ=True),ScaleFilter=(bX=True,bY=True,bZ=True))),(T ...(truncated)"
    bUseInitialTransforms=True
    Targets.0.Transform=&RigUnit_GetBoneTransform(Bone=spine_05).Transform
    Targets.1.Transform=&RigUnit_GetBoneTransform(Bone=clavicle_l).Transform
    Targets.2.Transform=&RigUnit_GetBoneTransform(Bone=upperarm_twist_01_l).Transform
    BaseTransform=&RigUnit_GetBoneTransform(Bone=dyn_shouderBelt_01_l_TGT).Transform
   )
   > &RigUnit_TransformConstraint(
    BaseTransformSpace=GlobalSpace
    Bone=dyn_shouderBelt_01_l_MSTR
    Targets="((Transform=(Rotation=(X=0,Y=0,Z=0,W=1),Translation=(X=0,Y=0,Z=0),Scale3D=(X=1,Y=1,Z=1)),Weight=0.8,bMaintainOffset=True,Filter=(TranslationFilter=(bX=True,bY=True,bZ=True),RotationFilter=(bX=True,bY=True,bZ=True),ScaleFilter=(bX=True,bY=True,bZ=True))),(T ...(truncated)"
    bUseInitialTransforms=True
    BaseTransform=&RigUnit_GetBoneTransform(Bone=dyn_shouderBelt_01_l_MSTR).Transform
    Targets.0.Transform=&RigUnit_GetBoneTransform(Bone=spine_05).Transform
    Targets.1.Transform=&RigUnit_GetBoneTransform(Bone=clavicle_l).Transform
   )
   > &RigUnit_AddBoneTransform(
    Bone=dyn_shouderBelt_01_l_TGT
    Transform=(Rotation=(X=0,Y=0,Z=0,W=1),Translation=(X=10,Y=0,Z=0),Scale3D=(X=1,Y=1,Z=1))
    Weight=1
    bPropagateToChildren=True
   )
   > &RigUnit_AimBone(
    Bone=dyn_shouderBelt_01_l
    DebugSettings=(bEnabled=true,Scale=10,WorldOffset=(Rotation=(X=0,Y=0,Z=0,W=1),Translation=(X=0,Y=0,Z=0),Scale3D=(X=1,Y=1,Z=1)))
    Primary=(Weight=1,Axis=(X=1,Y=0,Z=0),Target=(X=1,Y=0,Z=0),Kind=Location,Space="dyn_shouderBelt_01_l_TGT")
    Secondary=(Weight=1,Axis=(X=0,Y=1,Z=0),Target=(X=0,Y=1,Z=0),Kind=Direction,Space="dyn_shouderBelt_01_l_TGT")
    Weight=1
    bPropagateToChildren=True
   )
   > &RigUnit_TransformConstraint(
    BaseTransformSpace=GlobalSpace
    Bone=dyn_backPad_MSTR
    Targets="((Transform=(Rotation=(X=0,Y=0,Z=0,W=1),Translation=(X=0,Y=0,Z=0),Scale3D=(X=1,Y=1,Z=1)),Weight=0.85,bMaintainOffset=True,Filter=(TranslationFilter=(bX=True,bY=True,bZ=True),RotationFilter=(bX=True,bY=True,bZ=True),ScaleFilter=(bX=True,bY=True,bZ=True))),( ...(truncated)"
    bUseInitialTransforms=True
    BaseTransform=&RigUnit_GetBoneTransform(Bone=dyn_backPad_MSTR).Transform
    Targets.0.Transform=&RigUnit_GetBoneTransform(Bone=spine_05).Transform
    Targets.1.Transform=&RigUnit_GetBoneTransform(Bone=spine_04).Transform
   )
   > &RigUnit_TransformConstraint(
    BaseTransformSpace=GlobalSpace
    Bone=dyn_backPad_r_TGT
    Targets="((Transform=(Rotation=(X=0,Y=0,Z=0,W=1),Translation=(X=0,Y=0,Z=0),Scale3D=(X=1,Y=1,Z=1)),Weight=0.7,bMaintainOffset=True,Filter=(TranslationFilter=(bX=True,bY=True,bZ=True),RotationFilter=(bX=True,bY=True,bZ=True),ScaleFilter=(bX=True,bY=True,bZ=True))),(T ...(truncated)"
    bUseInitialTransforms=True
    BaseTransform=&RigUnit_GetBoneTransform(Bone=dyn_chestPad_r_TGT).Transform
    Targets.0.Transform=&RigUnit_GetBoneTransform(Bone=spine_05).Transform
    Targets.1.Transform=&RigUnit_GetBoneTransform(Bone=clavicle_r).Transform
    Targets.2.Transform=&RigUnit_GetBoneTransform(Bone=upperarm_twist_01_r).Transform
   )
   > &RigUnit_AimBone(
    Bone=dyn_backPad_r
    DebugSettings=(Scale=10,WorldOffset=(Rotation=(X=0,Y=0,Z=0,W=1),Translation=(X=0,Y=0,Z=0),Scale3D=(X=1,Y=1,Z=1)))
    Primary=(Weight=1,Axis=(X=1,Y=0,Z=0),Target=(X=1,Y=0,Z=0),Kind=Location,Space="dyn_backPad_r_TGT")
    Secondary=(Weight=1,Axis=(X=0,Y=1,Z=0),Target=(X=0,Y=1,Z=0),Kind=Location,Space="dyn_backPad_r_TGT")
    Weight=1
    bPropagateToChildren=True
   )
   > &RigUnit_TransformConstraint(
    BaseTransformSpace=GlobalSpace
    Bone=dyn_backPad_l_TGT
    Targets="((Transform=(Rotation=(X=0,Y=0,Z=0,W=1),Translation=(X=0,Y=0,Z=0),Scale3D=(X=1,Y=1,Z=1)),Weight=0.7,bMaintainOffset=True,Filter=(TranslationFilter=(bX=True,bY=True,bZ=True),RotationFilter=(bX=True,bY=True,bZ=True),ScaleFilter=(bX=True,bY=True,bZ=True))),(T ...(truncated)"
    bUseInitialTransforms=True
    BaseTransform=&RigUnit_GetBoneTransform(Bone=dyn_chestPad_l_TGT).Transform
    Targets.0.Transform=&RigUnit_GetBoneTransform(Bone=spine_05).Transform
    Targets.1.Transform=&RigUnit_GetBoneTransform(Bone=clavicle_l).Transform
    Targets.2.Transform=&RigUnit_GetBoneTransform(Bone=upperarm_twist_01_l).Transform
   )
   > &RigUnit_AimBone(
    Bone=dyn_backPad_l
    DebugSettings=(Scale=10,WorldOffset=(Rotation=(X=0,Y=0,Z=0,W=1),Translation=(X=0,Y=0,Z=0),Scale3D=(X=1,Y=1,Z=1)))
    Primary=(Weight=1,Axis=(X=1,Y=0,Z=0),Target=(X=1,Y=0,Z=0),Kind=Location,Space="dyn_backPad_l_TGT")
    Secondary=(Weight=1,Axis=(X=0,Y=1,Z=0),Target=(X=0,Y=1,Z=0),Kind=Location,Space="dyn_backPad_l_TGT")
    Weight=1
    bPropagateToChildren=True
   )
   > &RigUnit_TransformConstraint(
    BaseTransform=(Rotation=(X=0,Y=0,Z=0,W=1),Translation=(X=0,Y=0,Z=0),Scale3D=(X=1,Y=1,Z=1))
    BaseTransformSpace=GlobalSpace
    Bone=dyn_chestPad_r_TGT
    Targets="((Transform=(Rotation=(X=0,Y=0,Z=0,W=1),Translation=(X=0,Y=0,Z=0),Scale3D=(X=1,Y=1,Z=1)),Weight=0.12,bMaintainOffset=True,Filter=(TranslationFilter=(bX=True,bY=True,bZ=True),RotationFilter=(bX=True,bY=True,bZ=True),ScaleFilter=(bX=True,bY=True,bZ=True))),( ...(truncated)"
    bUseInitialTransforms=True
    Targets.0.Transform=&RigUnit_GetBoneTransform(Bone=spine_02).Transform
    Targets.1.Transform=&RigUnit_GetBoneTransform(Bone=spine_03).Transform
    Targets.2.Transform=&RigUnit_GetBoneTransform(Bone=spine_04).Transform
    Targets.3.Transform=&RigUnit_GetBoneTransform(Bone=spine_05).Transform
   )
   > &RigUnit_AimBone(
    Bone=dyn_chestPad_r
    DebugSettings=(bEnabled=true,Scale=10,WorldOffset=(Rotation=(X=0,Y=0,Z=0,W=1),Translation=(X=0,Y=0,Z=0),Scale3D=(X=1,Y=1,Z=1)))
    Primary=(Weight=1,Axis=(X=1,Y=0,Z=0),Target=(X=1,Y=0,Z=0),Kind=Location,Space="dyn_chestPad_r_TGT")
    Secondary=(Weight=1,Axis=(X=0,Y=0,Z=1),Target=(X=0,Y=0,Z=1),Kind=Location,Space="dyn_chestPad_r_TGT")
    Weight=1
    bPropagateToChildren=True
   )
   > &RigUnit_TransformConstraint(
    BaseTransform=(Rotation=(X=0,Y=0,Z=0,W=1),Translation=(X=0,Y=0,Z=0),Scale3D=(X=1,Y=1,Z=1))
    BaseTransformSpace=GlobalSpace
    Bone=dyn_chestPad_l_TGT
    Targets="((Transform=(Rotation=(X=0,Y=0,Z=0,W=1),Translation=(X=0,Y=0,Z=0),Scale3D=(X=1,Y=1,Z=1)),Weight=0.12,bMaintainOffset=True,Filter=(TranslationFilter=(bX=True,bY=True,bZ=True),RotationFilter=(bX=True,bY=True,bZ=True),ScaleFilter=(bX=True,bY=True,bZ=True))),( ...(truncated)"
    bUseInitialTransforms=True
    Targets.0.Transform=&RigUnit_GetBoneTransform(Bone=spine_02).Transform
    Targets.1.Transform=&RigUnit_GetBoneTransform(Bone=spine_03).Transform
    Targets.2.Transform=&RigUnit_GetBoneTransform(Bone=spine_04).Transform
    Targets.3.Transform=&RigUnit_GetBoneTransform(Bone=spine_05).Transform
   )
   > &RigUnit_AimBone(
    Bone=dyn_chestPad_l
    DebugSettings=(bEnabled=true,Scale=10,WorldOffset=(Rotation=(X=0,Y=0,Z=0,W=1),Translation=(X=0,Y=0,Z=0),Scale3D=(X=1,Y=1,Z=1)))
    Primary=(Weight=1,Axis=(X=1,Y=0,Z=0),Target=(X=1,Y=0,Z=0),Kind=Location,Space="dyn_chestPad_l_TGT")
    Secondary=(Weight=1,Axis=(X=0,Y=0,Z=1),Target=(X=0,Y=0,Z=1),Kind=Location,Space="dyn_chestPad_r_TGT")
    Weight=1
    bPropagateToChildren=True
   )
   &RigUnit_GetBoneTransform(Bone=upperarm_twist_01_r)
  ]}
]
hierarchy [
 root {loc=<0 0 0>, rot=<0 0 0 1>}
 pelvis {parent=root, loc=<0 -0.7071 93.1191>, rot=<0.0127 0.707 -0.0127 -0.707>}
 spine_01 {parent=pelvis, loc=<6.7494 -0.6388 0>, rot=<0 0 0.034 -0.9994>}
 spine_02 {parent=spine_01, loc=<7.2518 -0.035 0>, rot=<0 0 0.0238 -0.9997>}
 spine_03 {parent=spine_02, loc=<6.0104 0 0>, rot=<0 0 0.0149 -0.9999>}
 spine_04 {parent=spine_03, loc=<8.1024 0 0>, rot=<0 0 -0.0132 -0.9999>}
 spine_05 {parent=spine_04, loc=<10.0146 0 0>, rot=<0 0 -0.0015 -1>}
 clavicle_l {parent=spine_05, loc=<4.3575 -0.1466 -2.3223>, rot=<0.0465 -0.7289 -0.009 -0.6829>}
 upperarm_l {parent=clavicle_l, loc=<15.5111 0 0>, rot=<0.0226 -0.3234 -0.0088 -0.9459>}
 lowerarm_l {parent=upperarm_l, loc=<26.6514 0 0>, rot=<0.0391 -0.0099 -0.2453 -0.9686>}
 hand_l {parent=lowerarm_l, loc=<22.3491 0 0>, rot=<-0.7081 -0.0507 0.0975 -0.6975>}
 index_metacarpal_l {parent=hand_l, loc=<1.6535 0.0183 -2.96>, rot=<0.0158 -0.0124 -0.0173 -0.9996>}
 index_01_l {parent=index_metacarpal_l, loc=<8.3008 0.0357 -0.0524>, rot=<-0.0044 0.0485 0.0924 -0.9945>}
 index_02_l {parent=index_01_l, loc=<3.3587 -0.3004 0.0452>, rot=<-0.002 0.0027 0.3225 -0.9466>}
 index_03_l {parent=index_02_l, loc=<2.5578 -0.123 -0.0252>, rot=<-0.002 0.0067 0.2558 -0.9667>}
 middle_metacarpal_l {parent=hand_l, loc=<2.1856 -0.1725 -0.5984>, rot=<0.006 0.0168 -0.0329 -0.9993>}
 middle_01_l {parent=middle_metacarpal_l, loc=<7.3275 -0.2447 0.0129>, rot=<-0.0208 0.0042 0.0858 -0.9961>}
 middle_02_l {parent=middle_01_l, loc=<4.0123 0.1401 0.1175>, rot=<-0.0218 0.0258 0.3512 -0.9357>}
 middle_03_l {parent=middle_02_l, loc=<2.6077 0.0534 0.1373>, rot=<0.0082 0.0188 0.1708 -0.9851>}
 pinky_metacarpal_l {parent=hand_l, loc=<1.7616 -0.9753 3.3814>, rot=<-0.0397 0.0814 -0.069 -0.9935>}
 pinky_01_l {parent=pinky_metacarpal_l, loc=<7.1709 -0.4831 -0.1094>, rot=<-0.0654 -0.0035 0.2242 -0.9723>}
 pinky_02_l {parent=pinky_01_l, loc=<2.8274 0.0893 -0.1043>, rot=<0.0037 0.0011 0.2854 -0.9584>}
 pinky_03_l {parent=pinky_02_l, loc=<2.2046 0.0307 -0.0078>, rot=<-0.0002 0.0098 0.1445 -0.9895>}
 ring_metacarpal_l {parent=hand_l, loc=<2.2074 -0.3625 1.3638>, rot=<-0.0284 0.0722 -0.0458 -0.9959>}
 ring_01_l {parent=ring_metacarpal_l, loc=<7.0295 -0.6197 -0.1446>, rot=<-0.0489 -0.0252 0.1892 -0.9804>}
 ring_02_l {parent=ring_01_l, loc=<3.6174 0.4646 -0.1027>, rot=<0 0 0.2806 -0.9598>}
 ring_03_l {parent=ring_02_l, loc=<2.4152 0.1141 0.0199>, rot=<0 0 0.1189 -0.9929>}
 thumb_01_l {parent=hand_l, loc=<1.3118 -0.3191 -2.459>, rot=<0.6546 -0.4042 0.0664 -0.6354>}
 thumb_02_l {parent=thumb_01_l, loc=<5.9336 -0.0448 0.0401>, rot=<-0.0037 0.0213 0.298 -0.9543>}
 thumb_03_l {parent=thumb_02_l, loc=<2.2223 0.3682 -0.0907>, rot=<0.0118 -0.013 0.0066 -0.9998>}
 lowerarm_twist_01_l {parent=lowerarm_l, loc=<17.8549 0 0>, rot=<0 0 0 1>}
 lowerarm_twist_02_l {parent=lowerarm_l, loc=<9.6199 0 0>, rot=<0 0 0 1>}
 upperarm_twist_01_l {parent=upperarm_l, loc=<0.0041 0 0>, rot=<0 0 0 -1>}
 upperarm_twist_02_l {parent=upperarm_l, loc=<13.7941 0 0>, rot=<0 0 0 -1>}
 clavicle_r {parent=spine_05, loc=<4.3575 -0.1466 2.3223>, rot=<0.7289 0.0465 -0.6829 0.009>}
 upperarm_r {parent=clavicle_r, loc=<-15.5111 0 0>, rot=<0.0226 -0.3234 -0.0088 -0.9459>}
 lowerarm_r {parent=upperarm_r, loc=<-26.6514 0 0>, rot=<0.0391 -0.0099 -0.2453 -0.9686>}
 hand_r {parent=lowerarm_r, loc=<-22.3492 0 0>, rot=<-0.7081 -0.0507 0.0975 -0.6975>}
 index_metacarpal_r {parent=hand_r, loc=<-1.6534 -0.0181 2.9601>, rot=<0.0158 -0.0124 -0.0173 -0.9996>}
 index_01_r {parent=index_metacarpal_r, loc=<-8.3008 -0.0357 0.0524>, rot=<-0.0044 0.0485 0.0924 -0.9945>}
 index_02_r {parent=index_01_r, loc=<-3.3589 0.3003 -0.0453>, rot=<-0.002 0.0027 0.3225 -0.9466>}
 index_03_r {parent=index_02_r, loc=<-2.5578 0.1231 0.0251>, rot=<-0.002 0.0067 0.2558 -0.9667>}
 middle_metacarpal_r {parent=hand_r, loc=<-2.1856 0.1726 0.5984>, rot=<0.006 0.0168 -0.0329 -0.9993>}
 middle_01_r {parent=middle_metacarpal_r, loc=<-7.3273 0.2447 -0.0129>, rot=<-0.0208 0.0042 0.0858 -0.9961>}
 middle_02_r {parent=middle_01_r, loc=<-4.0142 -0.1411 -0.07>, rot=<-0.0217 0.0259 0.3512 -0.9357>}
 middle_03_r {parent=middle_02_r, loc=<-2.6074 -0.0534 -0.1312>, rot=<0.0083 0.0188 0.1708 -0.9851>}
 pinky_metacarpal_r {parent=hand_r, loc=<-1.7617 0.9753 -3.3814>, rot=<-0.0397 0.0814 -0.069 -0.9935>}
 pinky_01_r {parent=pinky_metacarpal_r, loc=<-7.1707 0.4832 0.1094>, rot=<-0.0654 -0.0035 0.2242 -0.9723>}
 pinky_02_r {parent=pinky_01_r, loc=<-2.8274 -0.0892 0.1043>, rot=<0.0037 0.0011 0.2854 -0.9584>}
 pinky_03_r {parent=pinky_02_r, loc=<-2.2045 -0.031 0.0078>, rot=<-0.0002 0.0098 0.1445 -0.9895>}
 ring_metacarpal_r {parent=hand_r, loc=<-2.2073 0.3627 -1.3638>, rot=<-0.0284 0.0722 -0.0458 -0.9959>}
 ring_01_r {parent=ring_metacarpal_r, loc=<-7.0297 0.6197 0.1446>, rot=<-0.0489 -0.0252 0.1892 -0.9804>}
 ring_02_r {parent=ring_01_r, loc=<-3.6172 -0.4648 0.1027>, rot=<0 0 0.2806 -0.9598>}
 ring_03_r {parent=ring_02_r, loc=<-2.4152 -0.1141 -0.0199>, rot=<0 0 0.1189 -0.9929>}
 thumb_01_r {parent=hand_r, loc=<-1.3119 0.3192 2.459>, rot=<0.6546 -0.4042 0.0664 -0.6354>}
 thumb_02_r {parent=thumb_01_r, loc=<-5.9335 0.0447 -0.0401>, rot=<-0.0037 0.0213 0.298 -0.9543>}
 thumb_03_r {parent=thumb_02_r, loc=<-2.2223 -0.3682 0.0907>, rot=<0.0118 -0.013 0.0066 -0.9998>}
 lowerarm_twist_01_r {parent=lowerarm_r, loc=<-17.855 0 0>, rot=<0 0 0 1>}
 lowerarm_twist_02_r {parent=lowerarm_r, loc=<-9.6199 0 0>, rot=<0 0 0 1>}
 upperarm_twist_01_r {parent=upperarm_r, loc=<-0.0041 0 0>, rot=<0 0 0 1>}
 upperarm_twist_02_r {parent=upperarm_r, loc=<-13.7942 0 0>, rot=<0 0 0 1>}
 neck_01 {parent=spine_05, loc=<11.3138 0.1529 0>, rot=<0 0 -0.2003 -0.9797>}
 neck_02 {parent=neck_01, loc=<4.5627 0 0>, rot=<0 0 0.0504 -0.9987>}
 head {parent=neck_02, loc=<5.3914 0 0>, rot=<0 0 0.115 -0.9934>}
 dyn_shouderBelt_01_r_MSTR {parent=spine_05, loc=<15.545 -0.3916 10.8357>, rot=<0.5982 0.0231 0.8003 -0.0336>}
 dyn_shouderBelt_01_r {parent=dyn_shouderBelt_01_r_MSTR, loc=<0 0 0>, rot=<0 0 0 -1>}
 dyn_shouderBelt_02_r_MSTR {parent=spine_05, loc=<8.5601 12.1515 10.2377>, rot=<0.482 0.3017 0.7297 -0.3797>}
 dyn_shouderBelt_02_r {parent=dyn_shouderBelt_02_r_MSTR, loc=<0 0 0>, rot=<0 0 0 1>}
 dyn_shouderBelt_01_l_MSTR {parent=spine_05, loc=<15.5447 -0.3916 -10.8357>, rot=<0.0344 -0.8213 -0.0219 -0.569>}
 dyn_shouderBelt_01_l {parent=dyn_shouderBelt_01_l_MSTR, loc=<0 0 0>, rot=<0 0 0 1>}
 dyn_shouderBelt_02_l_MSTR {parent=spine_05, loc=<7.9996 10.2256 -8.6155>, rot=<0.2809 -0.7245 -0.3082 -0.5488>}
 dyn_shouderBelt_02_l {parent=dyn_shouderBelt_02_l_MSTR, loc=<0 0 0>, rot=<0 0 0 -1>}
 dyn_chestPad_r_MSTR {parent=spine_05, loc=<1.8663 10.7288 8.5467>, rot=<0.0527 0.5569 0.8237 0.0924>}
 dyn_chestPad_r {parent=dyn_chestPad_r_MSTR, loc=<0 0 0>, rot=<0 0 0 -1>}
 dyn_padBelt_01 {parent=dyn_chestPad_r, loc=<13.8741 2.6865 0.4619>, rot=<0.014 -0.1221 -0.032 -0.9919>}
 dyn_padBelt_02 {parent=dyn_padBelt_01, loc=<8.25 0 0>, rot=<0.0008 -0.0124 0.0113 -0.9999>}
 dyn_padBelt_03 {parent=dyn_padBelt_02, loc=<7.1528 0 0>, rot=<0 0 0 -1>}
 dyn_backPad_MSTR {parent=spine_05, loc=<0.6197 -12.8181 0>, rot=<0.0284 -0.7065 -0.0284 -0.7065>}
 dyn_backPad_r {parent=dyn_backPad_MSTR, loc=<-1.2367 0.0895 0>, rot=<-0.0255 -0.7066 0.7066 0.0255>}
 dyn_backPad_l {parent=dyn_backPad_MSTR, loc=<1.2367 0.0896 0.0005>, rot=<-0.7066 -0.0255 -0.0255 -0.7066>}
 dyn_chestPad_l_MSTR {parent=spine_05, loc=<1.0398 11.1533 -9.4502>, rot=<-0.0685 -0.8212 -0.5639 -0.0535>}
 dyn_chestPad_l {parent=dyn_chestPad_l_MSTR, loc=<0 0 0>, rot=<0 0 0 1>}
 dyn_shouderBelt_01_r_TGT {parent=spine_05, loc=<13.5494 -0.5765 17.6207>, rot=<0.5982 0.0231 0.8003 -0.0336>}
 dyn_shouderBelt_02_r_TGT {parent=spine_05, loc=<7.1501 10.6483 15.5599>, rot=<0.482 0.3017 0.7297 -0.3797>}
 dyn_shouderBelt_01_l_TGT {parent=spine_05, loc=<13.1239 -0.6102 -17.3089>, rot=<0.0344 -0.8213 -0.0219 -0.569>}
 dyn_shouderBelt_02_l_TGT {parent=spine_05, loc=<6.7956 9.8799 -13.4786>, rot=<0.2809 -0.7245 -0.3082 -0.5488>}
 dyn_backPad_r_TGT {parent=spine_05, loc=<0.5341 -11.7532 14.7593>, rot=<0.5005 -0.4619 0.5381 0.4966>}
 dyn_backPad_l_TGT {parent=spine_05, loc=<0.5346 -11.7532 -14.7594>, rot=<-0.4966 -0.5381 0.4619 -0.5005>}
 dyn_chestPad_r_TGT {parent=spine_04, loc=<-2.0194 13.6845 8.3192>, rot=<0.0519 0.557 0.8239 0.0912>}
 dyn_chestPad_l_TGT {parent=spine_04, loc=<-1.811 13.3675 -9.5878>, rot=<-0.0672 -0.8213 -0.564 -0.0526>}
 thigh_l {parent=pelvis, loc=<-2.5693 -0.0109 -7.7501>, rot=<0.0012 0.0572 0.0155 -0.9982>}
 calf_l {parent=thigh_l, loc=<-39.0484 0 0>, rot=<0.0006 -0.0161 -0.0391 -0.9991>}
 calf_twist_01_l {parent=calf_l, loc=<-35.8131 -0.0967 0.0397>, rot=<0.0237 0.0095 -0.0035 -0.9997>}
 calf_twist_02_l {parent=calf_l, loc=<-18.8064 -0.0967 0.0397>, rot=<0.0237 0.0095 -0.0035 -0.9997>}
 foot_l {parent=calf_l, loc=<-42.561 0.1077 0.4292>, rot=<0.0418 -0.0455 0.0343 -0.9975>}
 ball_l {parent=foot_l, loc=<-6.1599 13.8264 0.6806>, rot=<0.0001 -0.0001 -0.7063 -0.7079>}
 thigh_twist_01_l {parent=thigh_l, loc=<-21.6303 0 0>, rot=<0.0013 0 0 -1>}
 thigh_r {parent=pelvis, loc=<-2.5691 -0.0109 7.7501>, rot=<0.0572 -0.0012 0.9982 0.0155>}
 calf_r {parent=thigh_r, loc=<39.0487 0 0>, rot=<0.0006 -0.0161 -0.0391 -0.9991>}
 calf_twist_01_r {parent=calf_r, loc=<35.8132 0.0967 -0.0397>, rot=<0.0237 0.0095 -0.0035 -0.9997>}
 calf_twist_02_r {parent=calf_r, loc=<18.8064 0.0967 -0.0397>, rot=<0.0237 0.0095 -0.0035 -0.9997>}
 foot_r {parent=calf_r, loc=<42.5611 -0.1077 -0.4292>, rot=<0.0418 -0.0455 0.0343 -0.9975>}
 ball_r {parent=foot_r, loc=<6.1598 -13.8264 -0.6806>, rot=<0.0001 -0.0001 -0.7063 -0.7079>}
 thigh_twist_01_r {parent=thigh_r, loc=<21.6303 0 0>, rot=<0.0013 0 0 -1>}
]
