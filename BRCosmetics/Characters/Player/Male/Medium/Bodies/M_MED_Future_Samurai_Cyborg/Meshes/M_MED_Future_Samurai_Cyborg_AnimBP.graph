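# Post-process style cosmetic AnimBP for the Future Samurai Cyborg body, copying the parent mesh pose and adding secondary motion.
# Constraints, look-ats and additive bone offsets correct straps, pads, bags, crotch, forearms and cape, driven by thigh, elbow and emote alphas.
# That cached pose feeds two spine-relative RigidBody cape simulations, selected by backbling state.
# Gravity follows swimming and skydiving; dynamics reset on movement transitions; the Control Rig node is unconnected.
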AnimBlueprint /BRCosmetics/Characters/Player/Male/Medium/Bodies/M_MED_Future_Samurai_Cyborg/Meshes/M_MED_Future_Samurai_Cyborg_AnimBP
target_skeleton=&/M_MED_Future_Samurai_Cyborg_Skeleton
graphs [
 AnimGraph {nodes [
   Control Rig() {
    alphaBoolBlend {blendOption=Linear}
    controlRigAssetReference {blueprintRigClass=/Game/Characters/Player/Male/Medium/Bodies/M_MED_Future_Samurai_Cyborg/Meshes/M_MED_Future_Samurai_Cyborg_CtrlRig.M_MED_Future_Samurai_Cyborg_CtrlRig_C}
   }
   n1 = RigidBody!(
    ComponentPose=Local To Component(LocalPose=Use cached pose 'PrePose'())
    Alpha
    OverrideWorldGravity:Vector=Get gravity_override() -> (gravity_override:Vector)
    ExternalForce
   ) -> (Pose) {
    bClampLinearTranslationLimitToRefPose=true
    bOverrideWorldGravity=true
    overridePhysicsAsset=&/M_MED_Future_Samurai_Cyborg_dyn_cape
    simulationSpace=BaseBoneSpace
    alphaBoolBlend {blendOption=Linear}
    baseBoneRef {boneName=spine_02}
    componentAppliedLinearAccClamp=<3000 3000 3000>
    componentLinearAccScale=<0.2 0.2 0.2>
    componentLinearVelScale=<0.2 0.2 0.2>
    overrideWorldGravity {z=-980}
   }
   n2 = RigidBody(ComponentPose=Local To Component(LocalPose=Use cached pose 'PrePose'()), OverrideWorldGravity=Get gravity_override()) {
    bClampLinearTranslationLimitToRefPose=true
    bOverrideWorldGravity=true
    overridePhysicsAsset=&/M_MED_Future_Samurai_Cyborg_dyn_cape_backbling
    simulationSpace=BaseBoneSpace
    alphaBoolBlend {blendOption=Linear}
    baseBoneRef {boneName=spine_02}
    componentAppliedLinearAccClamp=<3000 3000 3000>
    componentLinearAccScale=<0.15 0.15 0.15>
    componentLinearVelScale=<0.1 0.1 0.1>
    overrideWorldGravity {z=-980}
   }
   n3 = Blend Poses by bool(BlendPose_0=Component To Local(ComponentPose=@n2), BlendPose_1=Component To Local(ComponentPose=@n1), bActiveValue=Get isWearingBackbling()) {blendType=Linear}
   Output Pose(Result=Component To Local(ComponentPose=Local To Component(LocalPose=@n3))) {layerGroup=None}
   n4 = "Constraint - dyn_shouderBelt_01_r_MSTR (PPP)"!(
    ComponentPose=Local To Component(LocalPose=Copy Pose From Mesh() {bUseAttachedParent=true})
    Alpha=1
    ConstraintWeights_0=0.1851
    ConstraintWeights_1:float=0.0186
    ConstraintWeights_2:float=0.7963
   ) -> (Pose) {
    alphaBoolBlend {blendOption=Linear}
    boneToModify {boneName=dyn_shouderBelt_01_r_MSTR}
    constraintSetup [{targetBone {boneName=clavicle_r}}, {targetBone {boneName=upperarm_twist_01_r}}, {targetBone {boneName=spine_05}}]
    constraintWeights [1,1,1]
   }
   n5 = "Constraint - dyn_shouderBelt_01_r_TGT (PPP)"!(ComponentPose=@n4, Alpha=1, ConstraintWeights_0=0.5995, ConstraintWeights_1:float=0.1196, ConstraintWeights_2:float=0.2809) -> (Pose) {
    alphaBoolBlend {blendOption=Linear}
    boneToModify {boneName=dyn_shouderBelt_01_r_TGT}
    constraintSetup [{targetBone {boneName=clavicle_r}}, {targetBone {boneName=upperarm_twist_01_r}}, {targetBone {boneName=spine_05}}]
    constraintWeights [1,1,1]
   }
   n6 = Look At - Bone: dyn_shouderBelt_01_r(ComponentPose=@n5, Alpha=1) {
    bUseLookUpAxis=true
    alphaBoolBlend {blendOption=Linear}
    boneToModify {boneName=dyn_shouderBelt_01_r}
    lookAtLocation {x=20}
    lookAtTarget {boneReference {boneName=dyn_shouderBelt_01_r_TGT}}
    lookAt_Axis {axis=<1 0>}
    lookUp_Axis {axis=<0 1>}
   }
   n7 = "Constraint - dyn_chestPad_r_MSTR (PPP)"!(ComponentPose=@n6, Alpha=1, ConstraintWeights_0=0.1306, ConstraintWeights_1:float=0.0773, ConstraintWeights_2:float=0.7921) -> (Pose) {
    alphaBoolBlend {blendOption=Linear}
    boneToModify {boneName=dyn_chestPad_r_MSTR}
    constraintSetup [{targetBone {boneName=clavicle_r}}, {targetBone {boneName=spine_04}}, {targetBone {boneName=spine_05}}]
    constraintWeights [1,1,1]
   }
   n8 = "Constraint - dyn_chestPad_r_TGT (PPPPP)"!(
    ComponentPose=@n7
    Alpha=1
    ConstraintWeights_0=0.0018
    ConstraintWeights_1:float=0.1457
    ConstraintWeights_2:float=0.1253
    ConstraintWeights_3:float=0.3964
    ConstraintWeights_4:float=0.3307
   ) -> (Pose) {
    alphaBoolBlend {blendOption=Linear}
    boneToModify {boneName=dyn_chestPad_r_TGT}
    constraintSetup [
     {targetBone {boneName=clavicle_r}}
     {targetBone {boneName=spine_02}}
     {targetBone {boneName=spine_03}}
     {targetBone {boneName=spine_04}}
     {targetBone {boneName=spine_05}}
    ]
    constraintWeights [1,1,1,1,1]
   }
   n9 = Look At - Bone: dyn_chestPad_r(ComponentPose=@n8, Alpha=1) {
    bUseLookUpAxis=true
    alphaBoolBlend {blendOption=Linear}
    boneToModify {boneName=dyn_chestPad_r}
    lookAtLocation {x=10}
    lookAtTarget {boneReference {boneName=dyn_chestPad_r_TGT}}
    lookAt_Axis {axis=<1 0>}
    lookUp_Axis {axis=<0 1>}
   }
   n10 = "Constraint - dyn_bag_r_MSTR (PPPPP)"!(
    ComponentPose=@n9
    Alpha=1
    ConstraintWeights_0=0.2064
    ConstraintWeights_1:float=0.5003
    ConstraintWeights_2:float=0.1784
    ConstraintWeights_3:float=0.0377
    ConstraintWeights_4:float=0.0773
   ) -> (Pose) {
    alphaBoolBlend {blendOption=Linear}
    boneToModify {boneName=dyn_bag_r_MSTR}
    constraintSetup [
     {targetBone {boneName=spine_01}}
     {targetBone {boneName=spine_02}}
     {targetBone {boneName=spine_03}}
     {targetBone {boneName=spine_04}}
     {targetBone {boneName=pelvis}}
    ]
    constraintWeights [1,1,1,1,1]
   }
   n11 = "Constraint - dyn_bag_l_MSTR (PPPPP)"!(
    ComponentPose=@n10
    Alpha=1
    ConstraintWeights_0=0.2063
    ConstraintWeights_1:float=0.4936
    ConstraintWeights_2:float=0.1852
    ConstraintWeights_3:float=0.0378
    ConstraintWeights_4:float=0.0772
   ) -> (Pose) {
    alphaBoolBlend {blendOption=Linear}
    boneToModify {boneName=dyn_bag_l_MSTR}
    constraintSetup [
     {targetBone {boneName=spine_01}}
     {targetBone {boneName=spine_02}}
     {targetBone {boneName=spine_03}}
     {targetBone {boneName=spine_04}}
     {targetBone {boneName=pelvis}}
    ]
    constraintWeights [1,1,1,1,1]
   }
   n12 = Get thigh_up_l() -> (thigh_up_l:double)
   n13 = "Transform (Modify) Bone - Bone: dyn_leftCape_ft_MSTR"(ComponentPose=@n11, Alpha=@n12, Translation="0,0,0", Rotation="7.181,32.528,15", Scale="1,1,1") {
    rotationMode=BMM_Additive
    rotationSpace=BCS_BoneSpace
    translationMode=BMM_Additive
    translationSpace=BCS_BoneSpace
    alphaBoolBlend {blendOption=Linear}
    boneToModify {boneName=dyn_leftCape_ft_MSTR}
    rotation=<rot 7.181 32.528 15>
   }
   n14 = "Transform (Modify) Bone - Bone: dyn_rightCape_ft_MSTR"(ComponentPose=@n13, Alpha=@n12, Translation="0,0,0", Rotation="20.000,30,15", Scale="1,1,1") {
    rotationMode=BMM_Additive
    rotationSpace=BCS_BoneSpace
    translationMode=BMM_Additive
    translationSpace=BCS_BoneSpace
    alphaBoolBlend {blendOption=Linear}
    boneToModify {boneName=dyn_rightCape_ft_MSTR}
    rotation=<rot 20 30 15>
   }
   n15 = Get isPlayEmote() -> (isPlayEmote:double)
   n16 = "Transform (Modify) Bone - Bone: dyn_leftCape_ft_MSTR"(ComponentPose=@n14, Alpha=@n15, Translation="0,0,0", Rotation="10,0,0", Scale="1,1,1") {
    rotationMode=BMM_Additive
    rotationSpace=BCS_BoneSpace
    translationMode=BMM_Additive
    translationSpace=BCS_BoneSpace
    alphaBoolBlend {blendOption=Linear}
    boneToModify {boneName=dyn_leftCape_ft_MSTR}
    rotation {pitch=10}
   }
   n17 = "Transform (Modify) Bone - Bone: dyn_rightCape_ft_MSTR"(ComponentPose=@n16, Alpha=@n15, Translation="0,0,0", Rotation="10,0,0", Scale="1,1,1") {
    rotationMode=BMM_Additive
    rotationSpace=BCS_BoneSpace
    translationMode=BMM_Additive
    translationSpace=BCS_BoneSpace
    alphaBoolBlend {blendOption=Linear}
    boneToModify {boneName=dyn_rightCape_ft_MSTR}
    rotation {pitch=10}
   }
   n18 = "Transform (Modify) Bone - Bone: dyn_lowerarm_fix_r"(
    ComponentPose=@n17
    Alpha=Get elbow_rot_r() -> (elbow_rot_r:double)
    Translation="-2.5,0.000,0"
    Rotation="0.000,0.000,-35"
    Scale="1,0.8,0.8"
   ) {
    rotationMode=BMM_Additive
    rotationSpace=BCS_BoneSpace
    scaleMode=BMM_Additive
    scaleSpace=BCS_BoneSpace
    translationMode=BMM_Additive
    translationSpace=BCS_BoneSpace
    alphaBoolBlend {blendOption=Linear}
    boneToModify {boneName=dyn_lowerarm_fix_r}
    rotation {roll=-35}
    scale {y=0.8, z=0.8}
    translation {x=-2.5}
   }
   n19 = "Transform (Modify) Bone - Bone: dyn_lowerarm_fix_l"(
    ComponentPose=@n18
    Alpha=Get elbow_rot_l() -> (elbow_rot_l:double)
    Translation="2.5,0.000,0"
    Rotation="0.000,0.000,-35"
    Scale="1,0.8,0.8"
   ) {
    rotationMode=BMM_Additive
    rotationSpace=BCS_BoneSpace
    scaleMode=BMM_Additive
    scaleSpace=BCS_BoneSpace
    translationMode=BMM_Additive
    translationSpace=BCS_BoneSpace
    alphaBoolBlend {blendOption=Linear}
    boneToModify {boneName=dyn_lowerarm_fix_l}
    rotation {roll=-35}
    scale {y=0.8, z=0.8}
    translation {x=2.5}
   }
   n20 = "Constraint - dyn_crotch_aim_l (PPP)"!(ComponentPose=@n19, Alpha=1, ConstraintWeights_0=0.1, ConstraintWeights_1:float=0.45, ConstraintWeights_2:float=0.45) -> (Pose) {
    alphaBoolBlend {blendOption=Linear}
    boneToModify {boneName=dyn_crotch_aim_l}
    constraintSetup [{targetBone {boneName=pelvis}}, {targetBone {boneName=thigh_twist_01_l}}, {targetBone {boneName=thigh_twist_01_r}}]
    constraintWeights [1,1,1]
   }
   n21 = "Constraint - dyn_crotch_aim_r (PPP)"!(ComponentPose=@n20, Alpha=1, ConstraintWeights_0=0.1, ConstraintWeights_1:float=0.45, ConstraintWeights_2:float=0.45) -> (Pose) {
    alphaBoolBlend {blendOption=Linear}
    boneToModify {boneName=dyn_crotch_aim_r}
    constraintSetup [{targetBone {boneName=pelvis}}, {targetBone {boneName=thigh_twist_01_l}}, {targetBone {boneName=thigh_twist_01_r}}]
    constraintWeights [1,1,1]
   }
   n22 = Look At - Bone: dyn_crotch_l(ComponentPose=@n21, Alpha=1) {
    alphaBoolBlend {blendOption=Linear}
    boneToModify {boneName=dyn_crotch_l}
    lookAtLocation {x=0}
    lookAtTarget {boneReference {boneName=dyn_crotch_aim_l}}
    lookAt_Axis {axis=<1 0>}
    lookUp_Axis {axis=<0 1>}
   }
   n23 = Look At - Bone: dyn_crotch_r(ComponentPose=@n22, Alpha=1) {
    alphaBoolBlend {blendOption=Linear}
    boneToModify {boneName=dyn_crotch_r}
    lookAtLocation {x=0}
    lookAtTarget {boneReference {boneName=dyn_crotch_aim_r}}
    lookAt_Axis {axis=<1 0>}
    lookUp_Axis {axis=<0 1>}
   }
   Save cached pose 'PrePose'(Pose=Component To Local(ComponentPose=@n23)) {cacheName=PrePose}
  ]}
 EventGraph {nodes [
   Event BlueprintUpdateAnimation()
   > n1 = Sequence()
   n2 = GetAnimInstance(self=GetOwnerSkeletalMeshComponent())
   > Is Valid(InputObject=@n2)
   > Set isPlayEmote(isPlayEmote:double=SelectFloat(A=1, B=0, bPickA=Montage_IsPlaying(self=@n2, Montage=/BRCosmetics/Animation/Game/MainPlayer/Emotes/FutureSamurai/Emote_FutureSamurai_CMM_M))) -> (Output_Get:double)
   n3 = BreakTransform(InTransform=GetDeltaTransformFromRefPose(self=GetOwnerSkeletalMeshComponent(), BoneName=thigh_l, BaseName=pelvis))
   Set thigh_up_l(execute=@n1.then_1, thigh_up_l:double=MapRangeClamped(Value=BreakRotator(InRot=@n3.Rotation).Yaw, InRangeA=-20, InRangeB=-100)) -> (Output_Get:double)
   n4 = BreakTransform(InTransform=GetDeltaTransformFromRefPose(self=GetOwnerSkeletalMeshComponent(), BoneName=clavicle_r, BaseName=uppperarm_r))
   > Set arm_up_r(arm_up_r:double=MapRangeClamped(Value=BreakRotator(InRot=@n4.Rotation).Yaw, InRangeA=-20, InRangeB=-100)) -> (Output_Get:double)
   n5 = GetPartSkeletalMeshComponent()
   > n6 = TransformFromBoneSpace!(execute, self:Skinned Mesh Component Object Reference=@n5, BoneName:Name=hand_r, InPosition:Vector="0, 0, 0", InRotation:Rotator="0, 0, 0") -> (then, OutPosition:Vector, OutRotation:Rotator)
   > n7 = TransformFromBoneSpace(self=@n5, BoneName=upperarm_r, InPosition="0, 0, 0", InRotation="0, 0, 0")
   > Set elbow_rot_r(elbow_rot_r:double=MapRangeClamped(Value="Distance (Vector)"!(V1:Vector=@n6, V2:Vector=@n7) -> (ReturnValue:double), InRangeA=45, InRangeB=25)) -> (Output_Get:double)
   n8 = GetPartSkeletalMeshComponent()
   > n9 = TransformFromBoneSpace(self=@n8, BoneName=hand_l, InPosition="0, 0, 0", InRotation="0, 0, 0")
   > n10 = TransformFromBoneSpace(self=@n8, BoneName=upperarm_l, InPosition="0, 0, 0", InRotation="0, 0, 0")
   > Set elbow_rot_l(elbow_rot_l:double=MapRangeClamped(Value="Distance (Vector)"(V1=@n9, V2=@n10), InRangeA=45, InRangeB=25)) -> (Output_Get:double)
   > n11 = Branch(Condition=Get bIsSwimSprinting() -> (bIsSwimSprinting:bool))
   > Set gravity_override(gravity_override="0, 0, 0")
   n12 = Branch(execute=@n11.else, Condition=Get bIsSkyDiving())
   > n13 = Branch(Condition=Get bIsParachuteOpened())
   Set gravity_override(execute=[@n12.else, @n13.then], gravity_override="0, 0,-980")
   Set gravity_override(execute=@n13.else, gravity_override="0, 0,-300")
   n14 = Event OnBeginCrouching()
   ResetDynamics(execute=[@n16.then, @n14.then, @n17.then, @n18.then], InTeleportType=TeleportPhysics)
   > ResetDynamics(InTeleportType=ResetPhysics)
   Event OnCharacterCustomizationLoaded()
   > Do Once()
   n15 = CheckForBackbling()
   > Set isWearingBackbling(isWearingBackbling=OR Boolean(A=@n15, B=@n15.bIsBackblingLong))
   Event OnParachuteOpened()
   > n16 = Delay(Duration=1.5)
   n17 = Event OnEndSkydiving()
   Event BlueprintInitializeAnimation()
   > ResetDynamics(InTeleportType=TeleportPhysics)
   n18 = Event OnBeginSkydiving()
  ]}
]
variables [
 arm_up_r {type=number}
 elbow_rot_l {type=number}
 elbow_rot_r {type=number}
 gravity_override {type=Vector}
 isPlayEmote {type=number}
 isWearingBackbling {type=boolean, default=true}
 thigh_up_l {type=number}
]
